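; Passive ability for the S22 narrative plant-seeds quest that grants urgent transient quests in non-competitive Battle Royale.
; On activation it checks playlist and game phase, with a bus-landing event wait outside the required phase.
; After delays, it maps an in-progress main-quest objective to an urgent quest, applies a BeforeDamage effect and grants it.
; Completion re-runs granting, failure or no match ends the ability, and ending unbinds both delegates.
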(event Ability|EventFailedToActivatePassiveAbility
  (params (ActorInfo :type Gameplay%20Ability%20Actor%20Info%20Structure))
  (|CallParentFunction ActorInfo))

(event Ability|EventActivateAbility
  ;[0 0 0] Activate ability
  (|CallParentFunction)
  (CallFunction|DoPlaylistAndGamePhaseCheck))

(event Custom|Do%20Playlist%20and%20Game%20Phase%20Check
  ;[0 0 0] Ensure the player is in the right playlist and handle granting logic depending on current game phase
  (bind _bvalidplaylist (CallFunction|CheckIfValidPlaylist))
  (if _bvalidplaylist
    (bind _bisingameplayphase (CallFunction|CheckGamePhase))
    (if _bisingameplayphase
      (Development|PrintString "Already in gameplay phase, starting delay..." true true "(R=1.000000,G=0.000000,B=0.222520,A=1.000000)" 3.0)
      (else
        (Ability|Tasks|WaitGameplayEvent "(TagName=\"Event.SkydiveEnded.FromBus\")" 0 true
          (:then
            (Development|PrintString "Waiting for land from bus..." true true "(R=1.000000,G=0.000000,B=0.222520,A=1.000000)" 3.0))
          (:EventReceived
            (Development|PrintString "Landed from bus, starting delay..." true true "(R=1.000000,G=0.000000,B=0.222520,A=1.000000)" 3.0)))))
    (CallFunction|HandleUrgentQuestGranting)
    (else
      (Ability|EndAbility))))

(event Custom|Handle%20Urgent%20Quest%20Granting
  (bind _owningplayerpawn (Variables|GAVars|Data|GetOwningPlayerPawn))
  ;[0 0 0] Find the right urgent quest to grant based on objective, if it exists
  (Variables|Default|SetMainQuestItem (Quests|GetQuestwithDefinition (Variables|GAVars|Data|GetOwningPlayerQuestManager) "/BattlePassS22/Items/QuestItems/Narrative/Part_02/quest_s22_narrative_p02_q13.quest_s22_narrative_p02_q13"))
  (Utilities|FlowControl|Delay 2.0)
  (bind _bfoundquesttogrant (CallFunction|SetQuesttoGrantbyObjective))
  (if _bfoundquesttogrant
    (Utilities|FlowControl|Delay 1.0)
    ;[0 0 0] Apply Before Damage Condition and Grant the Urgent Quest
    (GameplayEffects|ApplyGameplayEffectToSelf :self (Ability|GetFortAbilitySystemComponent _owningplayerpawn) :GameplayEffectClass "/BattlePassS22/Items/QuestAbilities/Quest_Narrative_PlantSeeds/GE_Quest_BeforeDamage.GE_Quest_BeforeDamage_C" :Level 0.0)
    (Class|FortControllerComponentTransientQuests|GrantTransientQuest (Class|FortPlayerPawn|GetTransientQuestsComponent _owningplayerpawn) (Variables|Quest|GetUrgentQuestDefToGrant) 0)
    (CallFunction|BindUrgentQuestFailorComplete)
    (else
      (Ability|EndAbility))))

(event Custom|Bind%20Urgent%20Quest%20Fail%20Or%20Complete
  (bind _owningplayerquestmanager (Variables|GAVars|Data|GetOwningPlayerQuestManager))
  (bind _returnvalue (Class|FortQuestManager|GetQuestManagerComponent _owningplayerquestmanager "/Script/FortniteGame.FortQuestManagerComponent_Athena"))
  ;[0 0 0] If this urgent quest fails or is completed, see if there is another objective left to complete
  (Quests|BindEventto%28DEPRECATED%29OnUrgentTransientQuestFailed _returnvalue (delegate Custom|OnUrgentQuestFailed))
  (Quests|BindEventtoOnUrgentTransientQuestCompleted _returnvalue (delegate Custom|OnUrgentQuestComplete)))

(event Custom|OnUrgentQuestFailed
  (params (TransientQuestItem :type Fort%20Quest%20Item%20Object%20Reference))
  ;[0 0 0] If this urgent quest fails or is completed, see if there is another objective left to complete
  (if (== (AssetManager|GetPrimaryAssetIdfromObject (Quest|GetQuestDefinition TransientQuestItem)) (AssetManager|GetPrimaryAssetIdfromObject (Variables|Quest|GetUrgentQuestDefToGrant)))
    (Ability|EndAbility)))

(event Custom|OnUrgentQuestComplete
  (params (TransientQuestItem :type Fort%20Quest%20Item%20Object%20Reference))
  ;[0 0 0] If this urgent quest fails or is completed, see if there is another objective left to complete
  (if (== (AssetManager|GetPrimaryAssetIdfromObject (Quest|GetQuestDefinition TransientQuestItem)) (AssetManager|GetPrimaryAssetIdfromObject (Variables|Quest|GetUrgentQuestDefToGrant)))
    (CallFunction|HandleUrgentQuestGranting)))

(event Ability|EventOnEndAbility
  (params (bWasCancelled :type Boolean))
  (bind _owningplayerquestmanager (Variables|GAVars|Data|GetOwningPlayerQuestManager))
  (bind _returnvalue (Class|FortQuestManager|GetQuestManagerComponent _owningplayerquestmanager "/Script/FortniteGame.FortQuestManagerComponent_Athena"))
  ;[0 0 0] End ability - clean up binds
  (|CallParentFunction bWasCancelled)
  (Quests|UnbindEventfrom%28DEPRECATED%29OnUrgentTransientQuestFailed _returnvalue (EventDispatchers|CreateEvent :!event Custom|OnUrgentQuestFailed))
  (Quests|UnbindEventfromOnUrgentTransientQuestCompleted _returnvalue (EventDispatchers|CreateEvent :!event Custom|OnUrgentQuestComplete)))

(fn CheckifValidPlaylist
  ;[0 0 0] Only grant urgent quest if the player is loaded into Core BR
  (return (and (Class|FortGameStateAthena|IsBattleRoyalePlaylist (Fortnite|Athena|GameState|GetGameStateAthenafromContext (Variables|GAVars|Data|GetOwningPlayerPawn)) false) (not (Class|FortGameStateAthena|IsCompetitivePlaylist (Fortnite|Athena|GameState|GetGameStateAthenafromContext (Variables|GAVars|Data|GetOwningPlayerPawn)))))))

(fn CheckGamePhase
  ;[0 0 0] Check if the player is in the gameplay phase or on the bus
  (return (Utilities|Enum|Equal(Enum) (Variables|Default|GetRequiredGamePhase) (BattleRoyaleGamePhaseLogic|GetGamePhase (BattleRoyaleGamePhaseLogic|GetBattleRoyaleGamePhaseLogicComponent)))))

(fn SetQuestToGrantByObjective
  (bind _urgentquestdeftogrant (Variables|Quest|GetUrgentQuestDefToGrant))
  ;[0 0 0] Depending on hardcoded objective backend name vs urgent quest to grant, find the one that needs to be granted
  (Variables|Quest|SetUrgentQuestDefToGrant (Variables|Default|GetNullQuest))
  (bind _keys (Utilities|Map|Keys (Variables|Quest|GetObjectiveToUrgentQuest)))
  (bind _array_element (Utilities|Array|ForEachLoopwithBreak _keys)
    (:LoopBody
      (if (and (not (Quest|HasCompletedObjectivewithName (Variables|Default|GetMainQuestItem) _array_element :!class /Script/FortniteGame.FortQuestItem)) (Quest|IsObjectivewithNameinProgress (Variables|Default|GetMainQuestItem) _array_element :!class /Script/FortniteGame.FortQuestItem))
        (Variables|Quest|SetUrgentQuestDefToGrant (Utilities|Map|Find (Variables|Quest|GetObjectiveToUrgentQuest) _array_element))))
    (:Completed
      _))
  (Development|PrintString (Utilities|GetDisplayName _urgentquestdeftogrant) true true "(R=1.000000,G=0.000000,B=0.227117,A=1.000000)" 4.0)
  (return (!= _urgentquestdeftogrant (Variables|Default|GetNullQuest))))
