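; Gameplay ability for the urgent variant of the BattlePass S22 narrative Plant Seeds quest, handling quest failure when a different team is involved.
; On event activation it runs the parent logic, then compares the event's target player pawn with the instigator.
; If they are on different teams, it shows a failure message, fails the quest and ends itself.
; The message rebuilds the urgent quest definition's data with failure text and displays it through the Athena HUD context.
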(event Ability|EventActivateAbilityFromEvent
  (params (EventData :type Gameplay%20Event%20Data%20Structure%20%28by%20ref%29))
  (bind (_eventtag _instigator _target) (Utilities|Struct|BreakGameplayEventData EventData))
  (|CallParentFunction EventData)
  (bind _asfort_player_pawn (Utilities|Casting|CastToFortPlayerPawn _target))
  (if (not (Team|OnSameTeam _asfort_player_pawn _instigator))
    (CallFunction|ShowFailMessage)
    (CallFunction|FailQuest)
    (Ability|EndAbility)))

(event Custom|ShowFailMessage
  (bind (_eventtag _visualstyletag _totaleventtime) (Utilities|Struct|BreakUrgentQuestData (Class|FortUrgentQuestItemDefinition|GetUrgentQuestData (Variables|Default|GetUrgent%20Quest%20Definition))))
  (Class|AthenaHUDContext|ShowUrgentQuestMessage (BlueprintContext|GetAthenaHUDContext) (Utilities|Struct|MakeUrgentQuestData _eventtag "(GameplayTags=)" _totaleventtime true false false false false true "" (Variables|Default|GetTextQuestFailed) (Variables|Default|GetEvent%20Subtitle))))

(event Ability|EventOnEndAbility
  (params (bWasCancelled :type Boolean)))
