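; Gameplay ability for the Season 22 Plant Seeds narrative quest that stops quest progress for the rest of the match once the player takes damage.
; On its triggering gameplay event it calls the parent, then casts the event target to a Fort player pawn.
; If the instigator is not on that pawn's team, it removes the owner's effect granting Granted.Athena.Quest.NoDamage, then ends.
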(event Ability|EventActivateAbilityFromEvent
  (params (EventData :type Gameplay%20Event%20Data%20Structure%20%28by%20ref%29))
  (bind (_eventtag _instigator _target) (Utilities|Struct|BreakGameplayEventData EventData))
  ;[1 1 1] If the player takes damage, remove the no damage GE with granted tag so quest will no longer progress this match
  (|CallParentFunction EventData)
  (bind _asfort_player_pawn (Utilities|Casting|CastToFortPlayerPawn _target)
    (:then
      (if (not (Team|OnSameTeam _asfort_player_pawn _instigator))
        (Ability|RemoveGameplayEffectFromOwnerWithGrantedTags "(GameplayTags=((TagName=\"Granted.Athena.Quest.NoDamage\")))")))
    (:CastFailed
      _))
  (Ability|EndAbility))
